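# Sound cue for the reactive effect sound of the DyedDuelist character skin cosmetic.
# A Random node picks one of three equally weighted reactive wave variations each time it plays.
# Playback uses the basic foley attenuation settings with a boosted volume multiplier.
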SoundCue /BRCosmetics/Athena/Sounds/Character/Skin/DyedDuelist/Char_DyedDuelist_Reactive_Cue
root=Random_0
nodes [
 {id=Random_0, children [WavePlayer_0,WavePlayer_1,WavePlayer_2], properties {weights [1,1,1]}}
 {id=WavePlayer_0, properties {soundWaveAssetPtr=/Game/Athena/Sounds/Character/Skin/DyedDuelist/Char_DyedDuelist_Reactive_01}}
 {id=WavePlayer_1, properties {soundWaveAssetPtr=/Game/Athena/Sounds/Character/Skin/DyedDuelist/Char_DyedDuelist_Reactive_02}}
 {id=WavePlayer_2, properties {soundWaveAssetPtr=/Game/Athena/Sounds/Character/Skin/DyedDuelist/Char_DyedDuelist_Reactive_03}}
]
properties {
 attenuationSettings=&Foley_Attenuation_Basic_Att
 maxDistance=1536
 volumeMultiplier=2.25
}
